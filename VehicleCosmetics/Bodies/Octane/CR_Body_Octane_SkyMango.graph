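# Control Rig for the Octane SkyMango vehicle body, posing its body, steering, axle, brake and wheel bones from animator controls.
# Setup snaps controls to initial bone transforms and copies left control shapes to right ones.
# Forward solve steers front wheels, and optionally the steering wheel, from direction_ctrl, then constrains brake pads and aims axle pivots at wheel discs.
# Wheel spin adds root X translation to accumulated car_movement_ctrl Y; door node is unconnected.
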ControlRigBlueprint /VehicleCosmetics/Bodies/Octane/CR_Body_Octane_SkyMango
graphs [
 RigVMFunctionLibrary {links [], nodes []}
 RigVMModel {nodes [
   n1 = &RigUnit_PrepareForExecution()
   &RigUnit_SetTransform(
    Item=(Type=Control,Name="root_ctrl")
    Weight=1
    bInitial=true
    ExecutePin=@n1.ExecutePin
    Value=&RigUnit_GetTransform(Item=(Type=Bone,Name="root"), bInitial=true).Transform
   )
   > &RigUnit_SetTransform(
    Item=(Type=Control,Name="body_ctrl")
    Weight=1
    bInitial=true
    Value=&RigUnit_GetTransform(Item=(Type=Bone,Name="body"), bInitial=true).Transform
   )
   Make_Array = RigVMDispatchNode(Values="((Type=Control,Name=\"steering_wheel_ctrl\"),(Type=Control,Name=\"wheel_steering_fr_l_ctrl\"),(Type=Control,Name=\"wheel_steering_fr_r_ctrl\"),(Type=Control,Name=\"wheel_disc_fr_l_ctrl\"),(Type=Control,Name=\"wheel_disc_fr_r_ctrl\"),(Type=Control,Name=\"wheel_disc_bk ...(truncated)") {template="DISPATCH_RigVMDispatch_ArrayMake(in Values,out Array)"}
   > For_Each = RigVMDispatchNode(Array=@Make_Array.Array) {template="DISPATCH_RigVMDispatch_ArrayIterator(in Array,out Element,out Index,out Count,out Ratio)"}
   Replace = /Script/RigVM.RigVMFunction_NameReplace(Old=_ctrl, Name=@For_Each.Element.Name) {is=struct RigVMUnitNode, template="Replace::Execute(in Name,in Old,in New,out Result)"}
   RigVMFunction_ControlFlowBranch = &RigVMFunction_ControlFlowBranch(
    ExecuteContext=@For_Each.ExecuteContext
    Condition=/Script/ControlRig.RigUnit_ItemExists(Item=(Type=Bone,Name="None"), Item.Name=@Replace.Result) {is=struct RigVMUnitNode}.Exists
   )
   Contains = /Script/RigVM.RigVMFunction_Contains(Search=_l_ctrl, Name=@For_Each.Element.Name) {is=struct RigVMUnitNode, template="Contains::Execute(in Name,in Search,out Result)"}
   n2 = &RigVMFunction_ControlFlowBranch(Condition=@Contains.Result, ExecuteContext=@RigVMFunction_ControlFlowBranch.Completed)
   n3 = &RigVMFunction_NameReplace(New=_r_ctrl, Old=_l_ctrl, Name=@For_Each.Element.Name) {template="Replace::Execute(in Name,in Old,in New,out Result)"}
   n4 = &RigVMFunction_ControlFlowBranch(Condition=&RigUnit_ItemExists(Item=(Type=Control,Name="None"), Item.Name=@n3.Result).Exists, ExecuteContext=@n2.True)
   /Script/ControlRig.RigUnit_SetShapeTransform(
    Transform=/Script/ControlRig.RigUnit_GetShapeTransform(Control=@For_Each.Element.Name) {is=struct RigVMUnitNode}.Transform
    Control=@n3.Result
    ExecutePin=@n4.True
   ) {is=struct RigVMUnitNode}
   &RigUnit_SetTransform(
    Item=(Type=Control,Name="wheel_disc_fr_l_ctrl")
    Weight=1
    bInitial=true
    Item.Name=@For_Each.Element.Name
    Value=&RigUnit_GetTransform(Item=(Type=Bone,Name="wheelFront_l"), bInitial=true, Item.Name=@Replace.Result).Transform
    ExecutePin=@RigVMFunction_ControlFlowBranch.True
   )
   RigUnit_BeginExecution = &RigUnit_BeginExecution()
   RigUnit_GetTransform = &RigUnit_GetTransform(Item=(Type=Control,Name="root_ctrl"), bInitial=False)
   RigUnit_SetTransform = &RigUnit_SetTransform(
    Item=(Type=Bone,Name="root")
    Weight=1
    bInitial=False
    bPropagateToChildren=true
    ExecutePin=@RigUnit_BeginExecution.ExecutePin
    Value=@RigUnit_GetTransform.Transform
   )
   > &RigUnit_SetTransform(
    Item=(Type=Bone,Name="body")
    Weight=1
    bInitial=False
    bPropagateToChildren=true
    Value=&RigUnit_GetTransform(Item=(Type=Control,Name="body_ctrl"), bInitial=False).Transform
   )
   > &RigUnit_SetTransform(
    Item=(Type=Bone,Name="steering_wheel")
    Weight=1
    bInitial=False
    bPropagateToChildren=true
    Value=&RigUnit_GetTransform(Item=(Type=Control,Name="steering_wheel_ctrl"), bInitial=False).Transform
   )
   RigVMFunction_MathQuaternionToEuler = &RigVMFunction_MathQuaternionToEuler(
    RotationOrder=ZYX
    Value=&RigUnit_GetTransform(Item=(Type=Control,Name="direction_ctrl"), Space=LocalSpace, bInitial=False).Transform.Rotation
   )
   n5 = &RigVMFunction_MathQuaternionFromEuler(Euler=(X=0,Y=0,Z=0), Euler.Z=@RigVMFunction_MathQuaternionToEuler.Result.Z)
   > RigUnit_SetRotation = &RigUnit_SetRotation(Item=(Type=Control,Name="wheel_steering_fr_l_ctrl"), Weight=1, bInitial=False, Value=@n5.Result)
   > n6 = &RigUnit_SetRotation(Item=(Type=Control,Name="wheel_steering_fr_r_ctrl"), Weight=1, bInitial=False, Value=@n5.Result)
   RigUnit_GetBoolAnimationChannel = /Script/ControlRig.RigUnit_GetBoolAnimationChannel(Channel=Turn steering wheel, Control=direction_ctrl, bInitial=False) {is=struct RigVMUnitNode, template="GetAnimationChannel::Execute(out Value,in Control,in Channel,in bInitial)"}
   n7 = &RigVMFunction_ControlFlowBranch(Condition=@RigUnit_GetBoolAnimationChannel.Value, ExecuteContext=@n6.ExecutePin)
   &RigUnit_OffsetTransformForItem(
    Item=(Type=Bone,Name="steering_wheel")
    OffsetTransform=(Rotation=(X=0,Y=0,Z=0,W=1),Translation=(X=0,Y=0,Z=0),Scale3D=(X=1,Y=1,Z=1))
    Weight=1
    bPropagateToChildren=True
    ExecutePin=@n7.True
    OffsetTransform.Rotation=&RigVMFunction_MathQuaternionFromEuler(
     Euler=(X=0,Y=0,Z=0)
     Euler.X=/Script/RigVM.RigVMFunction_MathDoubleNegate(Value=@RigVMFunction_MathQuaternionToEuler.Result.Z) {is=struct RigVMUnitNode, template="Negate::Execute(in Value,out Result)"}.Result
    ).Result
   )
   /Script/ControlRig.RigUnit_PositionConstraintLocalSpaceOffset(
    Child=(Type=Control,Name="brakepad_fr_l_ctrl")
    Filter=(bX=True,bY=True,bZ=True)
    Parents=((Item=(Type=Control,Name="wheel_disc_fr_l_ctrl"),Weight=1))
    Weight=1
    bMaintainOffset=True
    ExecutePin=@n7.Completed
   ) {is=struct RigVMUnitNode}
   > &RigUnit_PositionConstraintLocalSpaceOffset(
    Child=(Type=Control,Name="brakepad_fr_r_ctrl")
    Filter=(bX=True,bY=True,bZ=True)
    Parents=((Item=(Type=Control,Name="wheel_disc_fr_r_ctrl"),Weight=1))
    Weight=1
    bMaintainOffset=True
   )
   > &RigUnit_PositionConstraintLocalSpaceOffset(
    Child=(Type=Control,Name="brakepad_bk_l_ctrl")
    Filter=(bX=True,bY=True,bZ=True)
    Parents=((Item=(Type=Control,Name="wheel_disc_bk_l_ctrl"),Weight=1))
    Weight=1
    bMaintainOffset=True
   )
   > &RigUnit_PositionConstraintLocalSpaceOffset(
    Child=(Type=Control,Name="brakepad_bk_r_ctrl")
    Filter=(bX=True,bY=True,bZ=True)
    Parents=((Item=(Type=Control,Name="wheel_disc_bk_r_ctrl"),Weight=1))
    Weight=1
    bMaintainOffset=True
   )
   > &RigUnit_RotationConstraintLocalSpaceOffset(
    AdvancedSettings=(InterpolationType=Average,RotationOrderForFilter=XZY)
    Child=(Type=Control,Name="brakepad_fr_l_ctrl")
    Filter=(bX=True,bY=True,bZ=True)
    Parents=((Item=(Type=Control,Name="wheel_steering_fr_l_ctrl"),Weight=1))
    Weight=1
   )
   > &RigUnit_RotationConstraintLocalSpaceOffset(
    AdvancedSettings=(InterpolationType=Average,RotationOrderForFilter=XZY)
    Child=(Type=Control,Name="brakepad_fr_r_ctrl")
    Filter=(bX=True,bY=True,bZ=True)
    Parents=((Item=(Type=Control,Name="wheel_steering_fr_r_ctrl"),Weight=1))
    Weight=1
   )
   > &RigUnit_AimConstraintLocalSpaceOffset(
    AimAxis=(X=0,Y=-1,Z=0)
    Child=(Type=Control,Name="axle_pivot_fr_l_ctrl")
    Filter=(bX=True,bY=false,bZ=false)
    Parents=((Item=(Type=Control,Name="wheel_disc_fr_l_ctrl"),Weight=1))
    Weight=1
    WorldUp=(Target=(X=0,Y=0,Z=1),Kind=Direction,Space=(Type=None,Name="None"))
   )
   > &RigUnit_AimConstraintLocalSpaceOffset(
    AimAxis=(X=0,Y=-1,Z=0)
    Child=(Type=Control,Name="axle_pivot_fr_r_ctrl")
    Filter=(bX=True,bY=false,bZ=false)
    Parents=((Item=(Type=Control,Name="wheel_disc_fr_r_ctrl"),Weight=1))
    Weight=1
    WorldUp=(Target=(X=0,Y=0,Z=1),Kind=Direction,Space=(Type=None,Name="None"))
   )
   > &RigUnit_AimConstraintLocalSpaceOffset(
    AimAxis=(X=0,Y=-1,Z=0)
    Child=(Type=Control,Name="axle_pivot_bk_l_ctrl")
    Filter=(bX=True,bY=false,bZ=false)
    Parents=((Item=(Type=Control,Name="wheel_disc_bk_l_ctrl"),Weight=1))
    Weight=1
    WorldUp=(Target=(X=0,Y=0,Z=1),Kind=Direction,Space=(Type=None,Name="None"))
   )
   > &RigUnit_AimConstraintLocalSpaceOffset(
    AimAxis=(X=0,Y=-1,Z=0)
    Child=(Type=Control,Name="axle_pivot_bk_r_ctrl")
    Filter=(bX=True,bY=false,bZ=false)
    Parents=((Item=(Type=Control,Name="wheel_disc_bk_r_ctrl"),Weight=1))
    Weight=1
    WorldUp=(Target=(X=0,Y=0,Z=1),Kind=Direction,Space=(Type=None,Name="None"))
   )
   Make_Array_2 = RigVMDispatchNode(Values=((Type=Control,Name="axle_pivot_fr_l_ctrl"),(Type=Control,Name="axle_pivot_fr_r_ctrl"),(Type=Control,Name="axle_pivot_bk_l_ctrl"),(Type=Control,Name="axle_pivot_bk_r_ctrl"))) {template="DISPATCH_RigVMDispatch_ArrayMake(in Values,out Array)"}
   > For_Each_2 = RigVMDispatchNode(Array=@Make_Array_2.Array) {template="DISPATCH_RigVMDispatch_ArrayIterator(in Array,out Element,out Index,out Count,out Ratio)"}
   n8 = &RigVMFunction_NameReplace(Old=_ctrl, Name=@For_Each_2.Element.Name) {template="Replace::Execute(in Name,in Old,in New,out Result)"}
   n9 = &RigVMFunction_ControlFlowBranch(
    ExecuteContext=@For_Each_2.ExecuteContext
    Condition=&RigUnit_ItemExists(Item=(Type=Bone,Name="None"), Item.Name=@n8.Result).Exists
   )
   &RigUnit_SetTransform(
    Item=(Type=Bone,Name="wheelFront_l_ctrl")
    Weight=1
    bInitial=false
    Item.Name=@n8.Result
    Value=&RigUnit_GetTransform(Item=(Type=Control,Name="wheelFront_l"), bInitial=false, Item.Name=@For_Each_2.Element.Name).Transform
    ExecutePin=@n9.True
   )
   Make_Array_3 = RigVMDispatchNode(Values=((Type=Control,Name="wheel_steering_fr_l_ctrl"),(Type=Control,Name="wheel_steering_fr_r_ctrl"))) {template="DISPATCH_RigVMDispatch_ArrayMake(in Values,out Array)"}
   For_Each_3 = RigVMDispatchNode(Array=@Make_Array_3.Array, ExecuteContext=@For_Each_2.Completed) {template="DISPATCH_RigVMDispatch_ArrayIterator(in Array,out Element,out Index,out Count,out Ratio)"}
   n10 = &RigVMFunction_ControlFlowBranch(
    ExecuteContext=@For_Each_3.ExecuteContext
    Condition=&RigUnit_ItemExists(
     Item=(Type=Bone,Name="None")
     Item.Name=&RigVMFunction_NameReplace(Old=_ctrl, Name=@For_Each_3.Element.Name) {template="Replace::Execute(in Name,in Old,in New,out Result)"}.Result
    ).Exists
   )
   &RigUnit_SetRotation(
    Item=(Type=Bone,Name="wheel_steering_fr_l")
    Space=GlobalSpace
    Weight=1
    bInitial=False
    Item.Name=@For_Each_3.Element.Name
    Value=&RigUnit_GetTransform(Item=(Type=Control,Name="wheelFront_l"), bInitial=false, Item.Name=@For_Each_3.Element.Name).Transform.Rotation
    ExecutePin=@n10.True
   )
   Make_Array_1 = RigVMDispatchNode(Values="((Type=Control,Name=\"wheel_disc_fr_l_ctrl\"),(Type=Control,Name=\"wheel_disc_fr_r_ctrl\"),(Type=Control,Name=\"wheel_disc_bk_l_ctrl\"),(Type=Control,Name=\"wheel_disc_bk_r_ctrl\"),(Type=Control,Name=\"brakepad_fr_l_ctrl\"),(Type=Control,Name=\"brakepad_fr_r_ctrl\"),( ...(truncated)") {template="DISPATCH_RigVMDispatch_ArrayMake(in Values,out Array)"}
   For_Each_1 = RigVMDispatchNode(Array=@Make_Array_1.Array, ExecuteContext=@For_Each_3.Completed) {template="DISPATCH_RigVMDispatch_ArrayIterator(in Array,out Element,out Index,out Count,out Ratio)"}
   n11 = &RigVMFunction_NameReplace(Old=_ctrl, Name=@For_Each_1.Element.Name) {template="Replace::Execute(in Name,in Old,in New,out Result)"}
   n12 = &RigVMFunction_ControlFlowBranch(
    ExecuteContext=@For_Each_1.ExecuteContext
    Condition=&RigUnit_ItemExists(Item=(Type=Bone,Name="None"), Item.Name=@n11.Result).Exists
   )
   &RigUnit_SetTransform(
    Item=(Type=Bone,Name="wheelFront_l_ctrl")
    Weight=1
    bInitial=false
    Item.Name=@n11.Result
    Value=&RigUnit_GetTransform(Item=(Type=Control,Name="wheelFront_l"), bInitial=false, Item.Name=@For_Each_1.Element.Name).Transform
    ExecutePin=@n12.True
   )
   Multiply = &RigVMFunction_MathDoubleMul(
    B=0.1
    A=&RigUnit_GetTransform(Item=(Type=Control,Name="car_movement_ctrl"), Space=LocalSpace, bInitial=False).Transform.Translation.Y
   )
   AccumulateAdd = /Script/RigVM.RigVMFunction_AccumulateFloatAdd(InitialValue=0, bIntegrateDeltaTime=False, Increment=@Multiply.Result) {is=struct RigVMUnitNode, template="AccumulateAdd::Execute(in Increment,in InitialValue,in bIntegrateDeltaTime,out Result)"}
   Add = &RigVMFunction_MathDoubleAdd(
    B=@AccumulateAdd.Result
    A=&RigUnit_GetTransform(Item=(Type=Control,Name="root_ctrl"), Space=LocalSpace, bInitial=False).Transform.Translation.X
   )
   RigVMFunction_MathQuaternionFromEuler = &RigVMFunction_MathQuaternionFromEuler(Euler=(X=0,Y=0,Z=0), Euler.Y=@Add.Result)
   &RigUnit_OffsetTransformForItem(
    Item=(Type=Bone,Name="wheel_disc_fr_l")
    OffsetTransform=(Rotation=(X=0,Y=0,Z=0,W=1),Translation=(X=0,Y=0,Z=0),Scale3D=(X=1,Y=1,Z=1))
    Weight=1
    bPropagateToChildren=True
    ExecutePin=@For_Each_1.Completed
    OffsetTransform.Rotation=@RigVMFunction_MathQuaternionFromEuler.Result
   )
   > &RigUnit_OffsetTransformForItem(
    Item=(Type=Bone,Name="wheel_disc_bk_l")
    OffsetTransform=(Rotation=(X=0,Y=0,Z=0,W=1),Translation=(X=0,Y=0,Z=0),Scale3D=(X=1,Y=1,Z=1))
    Weight=1
    bPropagateToChildren=True
    OffsetTransform.Rotation=@RigVMFunction_MathQuaternionFromEuler.Result
   )
   n13 = &RigVMFunction_MathQuaternionFromEuler(Euler=(X=0,Y=0,Z=0), Euler.Y=@Add.Result)
   > &RigUnit_OffsetTransformForItem(
    Item=(Type=Bone,Name="wheel_disc_fr_r")
    OffsetTransform=(Rotation=(X=0,Y=0,Z=0,W=1),Translation=(X=0,Y=0,Z=0),Scale3D=(X=1,Y=1,Z=1))
    Weight=1
    bPropagateToChildren=True
    OffsetTransform.Rotation=@n13.Result
   )
   > &RigUnit_OffsetTransformForItem(
    Item=(Type=Bone,Name="wheel_disc_bk_r")
    OffsetTransform=(Rotation=(X=0,Y=0,Z=0,W=1),Translation=(X=0,Y=0,Z=0),Scale3D=(X=1,Y=1,Z=1))
    Weight=1
    bPropagateToChildren=True
    OffsetTransform.Rotation=@n13.Result
   )
   n14 = &RigUnit_SetTransform(
    ExecutePin=()
    Item=(Type=Bone,Name="door_r")
    Weight=1
    bInitial=False
    bPropagateToChildren=true
    Value=&RigUnit_GetTransform(Item=(Type=Control,Name="door_r_ctrl"), bInitial=False).Transform
   )
   RigVMRerouteNode(Value=@n14.ExecutePin)
   Negate = &RigVMFunction_MathDoubleNegate(Value=@Add.Result) {template="Negate::Execute(in Value,out Result)"}
  ]}
]
hierarchy [
 root {loc=<0 0 0>, rot=<0 0 0 1>}
 frame {parent=root, loc=<-0.0698 0 14.166>, rot=<0 0 0 1>}
 body {parent=frame, loc=<0.0698 0 36.2744>, rot=<0 0 0 1>}
 glass_c {parent=body, loc=<134.5522 0 53.5066>, rot=<0 0 0 1>}
 glass_l {parent=body, loc=<52.6663 -54.1246 60.165>, rot=<0 0 0 1>}
 glass_mid_l {parent=body, loc=<30 -54.1246 60.165>, rot=<0 0 0 1>}
 glass_r {parent=body, loc=<52.6663 54.1246 60.165>, rot=<0 -1 0 0>}
 glass_mid_r {parent=body, loc=<30 54.1246 60.165>, rot=<0 -1 0 0>}
 steering_column {parent=body, loc=<133.8407 -27.7333 39.3319>, rot=<0 -0.1219 0 -0.9925>}
 steering_wheel {parent=steering_column, loc=<-46.386 0 0>, rot=<0 0 0 1>}
 wheel_well_spreader_fr_l {parent=body, loc=<158.7031 -99.7033 -18.2354>, rot=<0 0 0 1>}
 wheel_well_spreader_fr_r {parent=body, loc=<158.7031 99.7033 -18.2354>, rot=<0 0 0 1>}
 wheel_well_spreader_bk_l {parent=body, loc=<-115.1328 -109.1988 -12.9006>, rot=<-0.0441 -0.0001 0.0017 -0.999>}
 wheel_well_spreader_bk_r {parent=body, loc=<-115.1328 109.1988 -12.9006>, rot=<0.0441 0.0001 0.0017 -0.999>}
 shock_up_fr_l {parent=body, loc=<158.7031 -34.0854 19.8276>, rot=<0 0.3827 0.9239 0>}
 shock_up_fr_r {parent=body, loc=<158.7031 33.9484 19.9646>, rot=<0 -0.3827 0.9239 0>}
 shock_up_bk_l {parent=body, loc=<-115.1328 -59.5853 43.8592>, rot=<0 0.0872 0.9962 0>}
 shock_up_bk_r {parent=body, loc=<-115.1328 59.5853 43.8592>, rot=<0 -0.0872 0.9962 0>}
 cylinder_left {parent=body, loc=<-77.8648 -67.9099 48.5952>, rot=<-0.0834 -0.0066 -0.684 -0.7247>}
 cylinder_right {parent=body, loc=<-77.8686 67.9848 48.5952>, rot=<-0.0066 -0.0834 -0.7247 -0.684>}
 en_panel_bk_l_1 {parent=body, loc=<-159.5651 -28.8279 59.8671>, rot=<0 0 0 1>}
 en_panel_bk_l_2 {parent=body, loc=<-159.5651 -28.8279 44.9117>, rot=<0 0 0 -1>}
 en_panel_bk_l_3 {parent=body, loc=<-159.5651 -28.8279 29.9596>, rot=<0 0 0 1>}
 en_panel_bk_l_4 {parent=body, loc=<-159.5651 -28.8279 15.0074>, rot=<0 0 0 1>}
 en_panel_bk_l_5 {parent=body, loc=<-159.5651 -28.8279 0.0553>, rot=<0 0 0 1>}
 en_panel_bk_r_1 {parent=body, loc=<-159.5651 28.8279 59.8671>, rot=<0 0 0 1>}
 en_panel_bk_r_2 {parent=body, loc=<-159.5651 28.8279 44.9117>, rot=<0 0 0 1>}
 en_panel_bk_r_3 {parent=body, loc=<-159.5651 28.8279 29.9596>, rot=<0 0 0 1>}
 en_panel_bk_r_4 {parent=body, loc=<-159.5651 28.8279 15.0074>, rot=<0 0 0 1>}
 en_panel_bk_r_5 {parent=body, loc=<-159.5651 28.8279 0.0553>, rot=<0 0 0 1>}
 spoiler_wing {parent=body, loc=<-153.5273 0 123.4781>, rot=<0 0 0 1>}
 spoiler_shock_up {parent=spoiler_wing, loc=<19.3851 0 -4.1562>, rot=<0 -0.2752 0 -0.9614>}
 spoiler_shock_low {parent=body, loc=<-144.9707 0 101.9589>, rot=<0 -0.2752 0 -0.9614>}
 front_differential {parent=frame, loc=<158.773 0 18.0389>, rot=<0 0 0 1>}
 axle_pivot_fr_l {parent=front_differential, loc=<0 -21.4372 0>, rot=<0 0 0 1>}
 steering_knuckle_fr_l {parent=axle_pivot_fr_l, loc=<0 -28.4867 0>, rot=<0 0 0 1>}
 wheel_steering_fr_l {parent=steering_knuckle_fr_l, loc=<0 -30.9397 0>, rot=<0 0 0 1>}
 wheel_disc_fr_l {parent=wheel_steering_fr_l, loc=<0 -3.783 0>, rot=<0 0 0 1>}
 tire_fr_l {parent=wheel_disc_fr_l, loc=<0 7.5661 0>, rot=<0 0 0 1>}
 brakepad_fr_l {parent=wheel_steering_fr_l, loc=<0 -3.783 0>, rot=<0 0 0 1>}
 shock_base_fr_l {parent=steering_knuckle_fr_l, loc=<0 -9.7524 0>, rot=<0 0 0 1>}
 shock_low_fr_l {parent=shock_base_fr_l, loc=<0 0 12.4722>, rot=<0 0.3827 0.9239 0>}
 axle_pivot_fr_r {parent=front_differential, loc=<0 21.4372 0>, rot=<0 0 0 1>}
 steering_knuckle_fr_r {parent=axle_pivot_fr_r, loc=<0 28.4867 0>, rot=<0 0 0 1>}
 wheel_steering_fr_r {parent=steering_knuckle_fr_r, loc=<0 30.9397 0>, rot=<0 0 0 1>}
 wheel_disc_fr_r {parent=wheel_steering_fr_r, loc=<0 3.783 0>, rot=<0 0 0 -1>}
 tire_fr_r {parent=wheel_disc_fr_r, loc=<0 -7.5661 0>, rot=<0 0 0 1>}
 brakepad_fr_r {parent=wheel_steering_fr_r, loc=<0 3.783 0>, rot=<0 0 0 -1>}
 shock_base_fr_r {parent=steering_knuckle_fr_r, loc=<0 9.7524 0>, rot=<0 0 0 1>}
 shock_low_fr_r {parent=shock_base_fr_r, loc=<0 0 12.4722>, rot=<0 -0.3827 0.9239 0>}
 rear_differential {parent=frame, loc=<-115.0629 0 23.3737>, rot=<0 0 0 1>}
 axle_pivot_bk_l {parent=rear_differential, loc=<0 -27.1809 0>, rot=<0 0 0 1>}
 wheel_disc_bk_l {parent=axle_pivot_bk_l, loc=<0 -66.9574 0>, rot=<0 0 0 1>}
 tire_bk_l {parent=wheel_disc_bk_l, loc=<0 4.7478 0>, rot=<0 0 0 1>}
 brakepad_bk_l {parent=axle_pivot_bk_l, loc=<0 -66.9574 0>, rot=<0 0 0 1>}
 shock_base_bk_l {parent=axle_pivot_bk_l, loc=<0 -40.1376 0>, rot=<0 0 0 1>}
 shock_low_bk_l {parent=shock_base_bk_l, loc=<0 0 12.9022>, rot=<0 0.0872 0.9962 0>}
 axle_pivot_bk_r {parent=rear_differential, loc=<0 27.1809 0>, rot=<0 0 0 1>}
 wheel_disc_bk_r {parent=axle_pivot_bk_r, loc=<0 66.9574 0>, rot=<0 0 0 1>}
 tire_bk_r {parent=wheel_disc_bk_r, loc=<0 -4.7478 0>, rot=<0 0 0 1>}
 brakepad_bk_r {parent=axle_pivot_bk_r, loc=<0 66.9574 0>, rot=<0 0 0 1>}
 shock_base_bk_r {parent=axle_pivot_bk_r, loc=<0 40.1376 0>, rot=<0 0 0 1>}
 shock_low_bk_r {parent=shock_base_bk_r, loc=<0 0 12.9022>, rot=<0 -0.0872 0.9962 0>}
 root_ctrl {controlType=EULER_TRANSFORM, type=Control, loc=<0 0 0>, rot=<0 0 0 1>}
 car_movement_ctrl {controlType=EULER_TRANSFORM, parent=root_ctrl, type=Control, loc=<0 0 0>, rot=<0 0 0 1>}
 direction_ctrl {controlType=EULER_TRANSFORM, parent=root_ctrl, type=Control, loc=<0 0 0>, rot=<0 0 0 1>}
 Channel {controlType=BOOL, parent=direction_ctrl, type=Control, loc=<1 0 0>, rot=<0 0 0 1>}
 body_ctrl {controlType=EULER_TRANSFORM, parent=root_ctrl, type=Control, loc=<0 0 0>, rot=<0 0 0 1>}
 steering_wheel_ctrl {controlType=EULER_TRANSFORM, parent=body_ctrl, type=Control, loc=<0 0 0>, rot=<0 0 0 1>}
 axle_pivot_fr_l_ctrl {controlType=EULER_TRANSFORM, parent=body_ctrl, type=Control, loc=<0 0 0>, rot=<0 0 0 1>}
 axle_pivot_fr_r_ctrl {controlType=EULER_TRANSFORM, parent=body_ctrl, type=Control, loc=<0 0 0>, rot=<0 0 0 1>}
 axle_pivot_bk_l_ctrl {controlType=EULER_TRANSFORM, parent=body_ctrl, type=Control, loc=<0 0 0>, rot=<0 0 0 1>}
 axle_pivot_bk_r_ctrl {controlType=EULER_TRANSFORM, parent=body_ctrl, type=Control, loc=<0 0 0>, rot=<0 0 0 1>}
 door_r_ctrl {controlType=EULER_TRANSFORM, parent=body_ctrl, type=Control, loc=<0 0 0>, rot=<0 0 0 1>}
 wheel_steering_fr_l_ctrl {controlType=EULER_TRANSFORM, parent=root_ctrl, type=Control, loc=<0 0 0>, rot=<0 0 0 1>}
 wheel_disc_fr_l_ctrl {controlType=EULER_TRANSFORM, parent=wheel_steering_fr_l_ctrl, type=Control, loc=<0 0 0>, rot=<0 0 0 1>}
 wheel_steering_fr_r_ctrl {controlType=EULER_TRANSFORM, parent=root_ctrl, type=Control, loc=<0 0 0>, rot=<0 0 0 1>}
 wheel_disc_fr_r_ctrl {controlType=EULER_TRANSFORM, parent=wheel_steering_fr_r_ctrl, type=Control, loc=<0 0 0>, rot=<0 0 0 1>}
 wheel_disc_bk_l_ctrl {controlType=EULER_TRANSFORM, parent=root_ctrl, type=Control, loc=<0 0 0>, rot=<0 0 0 1>}
 wheel_disc_bk_r_ctrl {controlType=EULER_TRANSFORM, parent=root_ctrl, type=Control, loc=<0 0 0>, rot=<0 0 0 1>}
 brakepad_fr_l_ctrl {controlType=EULER_TRANSFORM, parent=root_ctrl, type=Control, loc=<0 0 0>, rot=<0 0 0 1>}
 brakepad_fr_r_ctrl {controlType=EULER_TRANSFORM, parent=root_ctrl, type=Control, loc=<0 0 0>, rot=<0 0 0 1>}
 brakepad_bk_l_ctrl {controlType=EULER_TRANSFORM, parent=root_ctrl, type=Control, loc=<0 0 0>, rot=<0 0 0 1>}
 brakepad_bk_r_ctrl {controlType=EULER_TRANSFORM, parent=root_ctrl, type=Control, loc=<0 0 0>, rot=<0 0 0 1>}
]
