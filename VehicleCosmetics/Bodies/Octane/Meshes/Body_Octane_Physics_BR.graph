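# Physics asset for the Octane vehicle body cosmetic, in a variant suffixed BR.
# It holds a single body on the root bone, with no collision shapes listed for it.
# It defines no constraints, so it sets up no articulated or jointed simulation.
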PhysicsAsset /VehicleCosmetics/Bodies/Octane/Meshes/Body_Octane_Physics_BR
bodies {root []}
constraints []
