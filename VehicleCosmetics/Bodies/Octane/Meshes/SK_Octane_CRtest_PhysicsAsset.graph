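# Physics asset for the Octane vehicle body skeletal mesh SK_Octane_CRtest, covering chassis, glass, spoiler, suspension, steering and wheels.
# Each bone gets a single capsule body, including one very large capsule on body and a tiny one on frame.
# Body links to frame and carries glass, steering wheel, spoiler and axle pivots with their shock bases, while steering knuckles attach to frame and carry wheel discs.
# All joints share identical swing and twist limit angles.
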PhysicsAsset /VehicleCosmetics/Bodies/Octane/Meshes/SK_Octane_CRtest_PhysicsAsset
bodies {
 axle_pivot_bk_l [{shapeType=Capsule, length=35.2964, radius=9.9974, center=<0.168 -28.4161 -0.2718>, rotation=<rot 0 179.827 -87.1975>}]
 axle_pivot_bk_r [{shapeType=Capsule, length=35.314, radius=9.9975, center=<0.1368 28.4178 -0.283>, rotation=<rot 0 0.1211 -87.0546>}]
 axle_pivot_fr_l [{shapeType=Capsule, length=26.5483, radius=9.3696, center=<-0.0547 -21.0803 -0.0237>, rotation=<rot 0 179.5641 -86.2455>}]
 axle_pivot_fr_r [{shapeType=Capsule, length=26.5354, radius=9.3649, center=<-0.0543 21.0808 -0.0229>, rotation=<rot 0 0.3868 -86.3765>}]
 body [{shapeType=Capsule, length=217.1677, radius=116.1213, center=<40.244 -0.5958 35.4266>, rotation=<rot 0 -90.2938 -82.4618>}]
 frame [{shapeType=Capsule, length=0.505, radius=0.505, rotation=<rot 0 -90 0>}]
 glass_c [{shapeType=Capsule, length=43.1903, radius=24.6883, center=<-17.9937 0.7853 10.0124>, rotation=<rot 0 -1.5107 -88.9509>}]
 glass_l [{shapeType=Capsule, length=58.8867, radius=13.749, center=<7.9441 5.9155 10.0877>, rotation=<rot 0 -89.7919 -74.8668>}]
 glass_r [{shapeType=Capsule, length=58.8867, radius=13.7491, center=<-7.9441 -5.9155 -10.0877>, rotation=<rot 0 -89.7919 -74.8668>}]
 shock_base_bk_l [{shapeType=Capsule, length=33.5821, radius=19.3387, center=<-0.2992 -0.7349 18.6471>, rotation=<rot 14.622 -90 -0.879>}]
 shock_base_bk_r [{shapeType=Capsule, length=33.7051, radius=19.3054, center=<-0.1993 0.6291 18.591>, rotation=<rot -15.2312 -90 -0.6332>}]
 shock_base_fr_l [{shapeType=Capsule, length=27.0691, radius=10.704, center=<-0.1438 5.2432 15.8697>, rotation=<rot 37.5882 -90 0.7888>}]
 shock_base_fr_r [{shapeType=Capsule, length=27.0662, radius=10.5409, center=<-0.1667 -5.2269 15.9065>, rotation=<rot -37.9675 -90 0.8869>}]
 spoiler_shock_low [{shapeType=Capsule, length=40.9604, radius=6.86, center=<-0.2437 -0.045 5.2118>, rotation=<rot 0 -180 -90>}]
 spoiler_shock_up [{shapeType=Capsule, length=40.8216, radius=9.9051, center=<-0.2586 -0.045 -8.302>, rotation=<rot 0 -180 -90>}]
 spoiler_wing [{shapeType=Capsule, length=95.6049, radius=36.8127, center=<-7.9686 0 2.9605>, rotation=<rot 0 180 -90>}]
 steering_knuckle_back_l [{shapeType=Capsule, length=16.997, radius=11.2818, center=<-4.9474 2.1259 3.6256>, rotation=<rot 0 -57.6425 -49.9031>}]
 steering_knuckle_back_r [{shapeType=Capsule, length=16.9599, radius=11.2898, center=<-4.9482 -2.1697 3.6271>, rotation=<rot 0 -122.0436 -49.8025>}]
 steering_knuckle_fr_l [{shapeType=Capsule, length=16.9038, radius=11.1958, center=<-4.6656 -1.4679 4.2152>, rotation=<rot -35.7472 -90 -36.3347>}]
 steering_knuckle_fr_r [{shapeType=Capsule, length=16.8742, radius=11.2042, center=<-4.6648 1.4315 4.2133>, rotation=<rot 35.4808 -90 -36.4262>}]
 steering_wheel [{shapeType=Capsule, length=17.6116, radius=13.7501, center=<2.5582 -0.2211 0.1568>, rotation=<rot 0 -179.9959 -89.9283>}]
 wheel_disc_bk_l [{shapeType=Capsule, length=19.4001, radius=19.3922, center=<0.0009 -2.0148 0.0095>, rotation=<rot 0.3179 -90 -15.4277>}]
 wheel_disc_bk_r [{shapeType=Capsule, length=19.4001, radius=19.3922, center=<0.0009 1.9249 0.0095>, rotation=<rot -0.3179 -90 -15.4275>}]
 wheel_disc_fr_l [{shapeType=Capsule, length=19.3937, radius=19.3858, center=<0 -1.8385 0.0095>, rotation=<rot 0.3179 -90 -15.4278>}]
 wheel_disc_fr_r [{shapeType=Capsule, length=19.3937, radius=19.3858, center=<0 1.7485 0.0095>, rotation=<rot -0.318 -90 -15.4279>}]
}
constraints [
 {bone1Name=body, bone2Name=frame, swing1LimitDegrees=45, swing2LimitDegrees=45, twistLimitDegrees=45}
 {bone1Name=steering_wheel, bone2Name=body, swing1LimitDegrees=45, swing2LimitDegrees=45, twistLimitDegrees=45}
 {bone1Name=glass_c, bone2Name=body, swing1LimitDegrees=45, swing2LimitDegrees=45, twistLimitDegrees=45}
 {bone1Name=glass_l, bone2Name=body, swing1LimitDegrees=45, swing2LimitDegrees=45, twistLimitDegrees=45}
 {bone1Name=glass_r, bone2Name=body, swing1LimitDegrees=45, swing2LimitDegrees=45, twistLimitDegrees=45}
 {bone1Name=spoiler_wing, bone2Name=body, swing1LimitDegrees=45, swing2LimitDegrees=45, twistLimitDegrees=45}
 {bone1Name=spoiler_shock_up, bone2Name=spoiler_wing, swing1LimitDegrees=45, swing2LimitDegrees=45, twistLimitDegrees=45}
 {bone1Name=spoiler_shock_low, bone2Name=body, swing1LimitDegrees=45, swing2LimitDegrees=45, twistLimitDegrees=45}
 {bone1Name=axle_pivot_fr_l, bone2Name=body, swing1LimitDegrees=45, swing2LimitDegrees=45, twistLimitDegrees=45}
 {bone1Name=shock_base_fr_l, bone2Name=axle_pivot_fr_l, swing1LimitDegrees=45, swing2LimitDegrees=45, twistLimitDegrees=45}
 {bone1Name=axle_pivot_fr_r, bone2Name=body, swing1LimitDegrees=45, swing2LimitDegrees=45, twistLimitDegrees=45}
 {bone1Name=shock_base_fr_r, bone2Name=axle_pivot_fr_r, swing1LimitDegrees=45, swing2LimitDegrees=45, twistLimitDegrees=45}
 {bone1Name=axle_pivot_bk_l, bone2Name=body, swing1LimitDegrees=45, swing2LimitDegrees=45, twistLimitDegrees=45}
 {bone1Name=shock_base_bk_l, bone2Name=axle_pivot_bk_l, swing1LimitDegrees=45, swing2LimitDegrees=45, twistLimitDegrees=45}
 {bone1Name=axle_pivot_bk_r, bone2Name=body, swing1LimitDegrees=45, swing2LimitDegrees=45, twistLimitDegrees=45}
 {bone1Name=shock_base_bk_r, bone2Name=axle_pivot_bk_r, swing1LimitDegrees=45, swing2LimitDegrees=45, twistLimitDegrees=45}
 {bone1Name=steering_knuckle_fr_l, bone2Name=frame, swing1LimitDegrees=45, swing2LimitDegrees=45, twistLimitDegrees=45}
 {bone1Name=wheel_disc_fr_l, bone2Name=steering_knuckle_fr_l, swing1LimitDegrees=45, swing2LimitDegrees=45, twistLimitDegrees=45}
 {bone1Name=steering_knuckle_fr_r, bone2Name=frame, swing1LimitDegrees=45, swing2LimitDegrees=45, twistLimitDegrees=45}
 {bone1Name=wheel_disc_fr_r, bone2Name=steering_knuckle_fr_r, swing1LimitDegrees=45, swing2LimitDegrees=45, twistLimitDegrees=45}
 {bone1Name=steering_knuckle_back_l, bone2Name=frame, swing1LimitDegrees=45, swing2LimitDegrees=45, twistLimitDegrees=45}
 {bone1Name=wheel_disc_bk_l, bone2Name=steering_knuckle_back_l, swing1LimitDegrees=45, swing2LimitDegrees=45, twistLimitDegrees=45}
 {bone1Name=steering_knuckle_back_r, bone2Name=frame, swing1LimitDegrees=45, swing2LimitDegrees=45, twistLimitDegrees=45}
 {bone1Name=wheel_disc_bk_r, bone2Name=steering_knuckle_back_r, swing1LimitDegrees=45, swing2LimitDegrees=45, twistLimitDegrees=45}
]
