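# Physics asset for the Octane vehicle body skeletal mesh, giving it a simplified box collision shape.
# A single body on the root bone holds one box sized to the car, offset forward and raised from the bone origin.
# There are no constraints, so the whole vehicle collides as one rigid shape.
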PhysicsAsset /VehicleCosmetics/Bodies/Octane/Meshes/SK_VEH_Octane_BoxPhysics
bodies {root [{shapeType=Box, extentX=444, extentY=216, extentZ=116, shapeName=root_box, center=<33.9857 0 99.28>}]}
constraints []
